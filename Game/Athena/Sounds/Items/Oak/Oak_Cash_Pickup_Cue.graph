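# Sound cue for the Oak cash item pickup, played with weapon pickup attenuation and the gadgets/items sound class.
# A quality level node sends the first two levels to a random pick of three pickup waves and the last two to a single wave.
# Both branches pass through modulators that hold pitch fixed.
# A concurrency override prevents new instances once the limit is reached.
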SoundCue /Game/Athena/Sounds/Items/Oak/Oak_Cash_Pickup_Cue
root=QualityLevel_0
nodes [
 {id=Modulator_2, children [Random_0], properties {pitchMax=1, pitchMin=1}}
 {id=Modulator_3, children [WavePlayer_4], properties {pitchMax=1, pitchMin=1}}
 {id=QualityLevel_0, children [Modulator_2,Modulator_2,Modulator_3,Modulator_3]}
 {id=Random_0, children [WavePlayer_2,WavePlayer_3,WavePlayer_1], properties {weights [1,1,1]}}
 {id=WavePlayer_1, properties {soundWaveAssetPtr=&/Oak_Cash_Pickup_01}}
 {id=WavePlayer_2, properties {soundWaveAssetPtr=&/Oak_Cash_Pickup_02}}
 {id=WavePlayer_3, properties {soundWaveAssetPtr=&/Oak_Cash_Pickup_03}}
 {id=WavePlayer_4, properties {soundWaveAssetPtr=&Oak_Cash_Pickup_01}}
]
properties {
 attenuationSettings=&Fort_Weapon_Pickup_Att
 bOverrideConcurrency=true
 maxDistance=3500
 soundClassObject=&Fort_Gadgets_Items
 volumeMultiplier=0.5
 concurrencyOverrides {resolutionRule=PreventNew}
}
